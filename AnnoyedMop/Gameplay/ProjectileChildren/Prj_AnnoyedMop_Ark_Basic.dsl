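; Basic "Ark" projectile Blueprint for the AnnoyedMop weapon content, extending a parent projectile with a three-segment layout and impact effects.
; On actor hit it runs the parent logic, then a cosmetic-only FX event spawns a scaled Niagara impact aligned to the surface normal.
; The hit component's collision object type selects a water or fire impact system.
; Overridden functions supply three segment initializers and transforms: one centered, two offset and angled to either side.
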(event EventBeginPlay
  (|CallParentFunction))

(event Collision|EventActorBeginOverlap
  (params (OtherActor :type Actor%20Object%20Reference)))

(event EventTick
  (params (DeltaSeconds :type Float%20%28single-precision%29))
  (|CallParentFunction DeltaSeconds) ; This node is disabled and will not be called.\nDrag off pins to build functionality.
)

(event EventOnHitActor
  (params (Actor :type Actor%20Object%20Reference) (Hit :type Hit%20Result%20Structure))
  (|CallParentFunction Actor Hit)
  (CallFunction|OnHitActorFX Hit))

(event Custom|OnHitActor_FX
  (params (Hit :type Hit%20Result%20Structure%20%28by%20ref%29))
  (bind (_bblockinghit _binitialoverlap _time _distance _location _impactpoint _normal _impactnormal _physmat _hitactor _hitcomponent) (Collision|BreakHitResult Hit))
  (Networking|CanExecuteCosmeticEvents
    (:True
      (if (Utilities|Enum|Equal(Enum) (Collision|GetCollisionObjectType _hitcomponent) "ECC_GameTraceChannel18")
        (Variables|Default|SetFX_ImpactSystem "/Game/Weapons/FORT_Generic_Effects/Impacts/Water/NS_Impact_Water_Athena/NS_Impact_Water_Beeg_Apollo.NS_Impact_Water_Beeg_Apollo")
        (else
          (Variables|Default|SetFX_ImpactSystem "/AnnoyedMop/Assets/VFX/NS_AnnoyedMop_Fire_Impact1.NS_AnnoyedMop_Fire_Impact1")))
      (bind _returnvalue (Niagara|SpawnSystematLocation (Variables|Default|GetFX_ImpactSystem) _impactpoint (Math|Conversions|RotationFromXVector _impactnormal) "1.000000,1.000000,1.000000" true true "AutoRelease"))
      (Niagara|SetNiagaraVariable(Float) _returnvalue "GlobalScale" (Variables|Default|GetFXImpactScale)))))

(fn FXGetSegmentParams
  (return 3 "1.000000,1.000000,1.000000" 30))

(fn FXGetSegmentTransforms
  (return (Utilities|Array|MakeArray (Math|Transform|MakeTransform "-11.000000,-50.000000,0.000000" "0.000000,-25.000000,0.000000") (Math|Transform|MakeTransform "0, 0, 0") (Math|Transform|MakeTransform "-11.000000,50.000000,0.000000" "0.000000,25.000000,0.000000"))))

(fn GetSegmentInitializers
  (return (Utilities|Array|MakeArray (Variables|Default|GetSegmentInitializer2_DestroyedAfterInit_DontAttach) (Variables|Default|GetSegmentInitializer_DestroyedAfterInit_DontAttach) (Variables|Default|GetSegmentInitializer1_DestroyedAfterInit_DontAttach))))

(fn ConstructionScript
  (|CallParentFunction))
