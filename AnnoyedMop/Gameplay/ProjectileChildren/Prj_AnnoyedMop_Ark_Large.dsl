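; Large arc ("Ark") projectile variant for AnnoyedMop, a child Blueprint that configures inherited multi-segment behavior through function overrides.
; FX overrides define four segments with a uniform vector parameter, placed and yawed symmetrically along a curved arc.
; A further override supplies four segment initializer components, named as unattached and destroyed after init.
; The construction script enlarges the impact FX scale, while BeginPlay, overlap and Tick are unused stubs.
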(event EventBeginPlay
  (|CallParentFunction) ; This node is disabled and will not be called.\nDrag off pins to build functionality.
)

(event Collision|EventActorBeginOverlap
  (params (OtherActor :type Actor%20Object%20Reference))
  (|CallParentFunction OtherActor) ; This node is disabled and will not be called.\nDrag off pins to build functionality.
)

(event EventTick
  (params (DeltaSeconds :type Float%20%28single-precision%29))
  (|CallParentFunction DeltaSeconds) ; This node is disabled and will not be called.\nDrag off pins to build functionality.
)

(fn FXGetSegmentParams
  (return 4 (Math|Conversions|ToVector(Float) 3.0) 45))

(fn FXGetSegmentTransforms
  (return (Utilities|Array|MakeArray (Math|Transform|MakeTransform "-35.000000,-117.000000,0.000000" "0.000000,-37.000000,0.000000") (Math|Transform|MakeTransform "-2.000000,-41.000000,0.000000" "0.000000,-10.000000,0.000000") (Math|Transform|MakeTransform "-2.000000,41.000000,0.000000" "0.000000,10.000000,0.000000") (Math|Transform|MakeTransform "-35.000000,117.000000,0.000000" "0.000000,37.000000,0.000000"))))

(fn GetSegmentInitializers
  (return (Utilities|Array|MakeArray (Variables|Default|GetSegmentInitializer_DestroyedAfterInit_DontAttach) (Variables|Default|GetSegmentInitializer2_DestroyedAfterInit_DontAttach) (Variables|Default|GetSegmentInitializer3_DestroyedAfterInit__DontAttach) (Variables|Default|GetSegmentInitializer1_DestroyedAfterInit_DontAttach))))

(fn ConstructionScript
  (|CallParentFunction)
  (Variables|Default|SetFXImpactScale 3.0))
