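; Massive variant of the AnnoyedMop Ark projectile, a child Blueprint that adds knockback and a wide segmented FX shape.
; On actor hit, if the Knockback_Enabled scalable float is true, it launches the player pawn along the projectile's forward vector with upward bias, harder the closer it is to the instigator.
; FX overrides supply six segment transforms in a curved, fanned arc with matching initializers, and construction enlarges the impact FX.
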(event EventBeginPlay
  (|CallParentFunction) ; This node is disabled and will not be called.\nDrag off pins to build functionality.
)

(event Collision|EventActorBeginOverlap
  (params (OtherActor :type Actor%20Object%20Reference))
  (|CallParentFunction OtherActor) ; This node is disabled and will not be called.\nDrag off pins to build functionality.
)

(event EventTick
  (params (DeltaSeconds :type Float%20%28single-precision%29))
  (|CallParentFunction DeltaSeconds) ; This node is disabled and will not be called.\nDrag off pins to build functionality.
)

(event EventOnHitActor
  (params (Actor :type Actor%20Object%20Reference) (Hit :type Hit%20Result%20Structure))
  (|CallParentFunction Actor Hit)
  (if (ScalableFloat|GetValueasBool (Variables|ScalableFloats|GetROW_Knockback_Enabled))
    (bind _asplayer_pawn_athena (Utilities|Casting|CastToPlayerPawn_Athena Actor))
    (Character|LaunchCharacter _asplayer_pawn_athena (* (+ (Transformation|GetActorForwardVector) "0.000000,0.000000,0.500000") (Utilities|RemapFloat (Math|Vector|VectorLength (- (Transformation|GetActorLocation (Variables|Actor|GetInstigator)) (Transformation|GetActorLocation _asplayer_pawn_athena))) "(X=4000.000000,Y=200.000000)" "(X=1100.000000,Y=1800.000000)" true)) true true)))

(fn FXGetSegmentParams
  (return 6 (Math|Conversions|ToVector(Float) 3.0) 59))

(fn FXGetSegmentTransforms
  (return (Utilities|Array|MakeArray (Math|Transform|MakeTransform "-104.000000,-264.000000,0.000000" "0.000000,-40.000000,0.000000") (Math|Transform|MakeTransform "-41.000000,-168.000000,0.000000" "0.000000,-26.000000,0.000000") (Math|Transform|MakeTransform "-5.000000,-58.000000,0.000000" "0.000000,-10.000000,0.000000") (Math|Transform|MakeTransform "-5.000000,58.000000,0.000000" "0.000000,10.000000,0.000000") (Math|Transform|MakeTransform "-41.000000,168.000000,0.000000" "0.000000,26.000000,0.000000") (Math|Transform|MakeTransform "-104.000000,264.000000,0.000000" "0.000000,40.000000,0.000000"))))

(fn GetSegmentInitializers
  (return (Utilities|Array|MakeArray (Variables|Default|GetSegmentInitializer_DestroyedAfterInit_DontAttach) (Variables|Default|GetSegmentInitializer4_DestroyedAfterInit_DontAttach) (Variables|Default|GetSegmentInitializer2_DestroyedAfterInit_DontAttach) (Variables|Default|GetSegmentInitializer3_DestroyedAfterInit__DontAttach) (Variables|Default|GetSegmentInitializer5_DestroyedAfterInit_DontAttach) (Variables|Default|GetSegmentInitializer1_DestroyedAfterInit_DontAttach))))

(fn ConstructionScript
  (|CallParentFunction)
  (Variables|Default|SetFXImpactScale 4.0))
